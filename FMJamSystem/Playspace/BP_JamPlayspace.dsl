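; Jam playspace Blueprint that connects jam session music to its bound BP_JamPlayspace_Volume and handles jam shutdown.
; Cosmetic-capable instances listen for loop and beat events, forwarding beat and bar to the volume for reactive VFX; tempo is pushed at setup, as its listener is unwired.
; Authority tracks server performance and can end non-persistent jams through the volume.
; SetMetasoundAttenuation and the loop-stopped handler are empty; stutter FX functions set MetaSound parameters for debugging.
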(event EventOnUserAdded
  (params (User :type Playspace%20User%20Structure%20%28by%20ref%29) (PlayerState :type Player%20State%20Object%20Reference))
  (bind (_userid _leafplayspace _playerstatecached) (Utilities|Struct|BreakPlayspaceUser User))
  ;[0.9454 0.3569 1] Players added and removed to the playspace
  (Development|PrintString (Utilities|String|Append (PlayerState|GetPlayerName _playerstatecached) " Entered " (Utilities|GetDisplayName self)) false true "(R=0.000000,G=0.640625,B=0.000000,A=1.000000)" 1e+01 ""))

(event Custom|OnBeat
  (params (Beat%20Number :type Integer) (Beat%20In%20Bar :type Integer))
  ;[1 0.573 0.4931] Reactive VFX
  (if (and (CallFunction|IsAnyMusicPlaying) (>= Beat%20Number 0))
    (Class|BPJamPlayspaceVolume|OnBeat (Variables|Default|GetPlayspaceVolume) Beat%20In%20Bar (+ (/ (- (+ Beat%20Number 1) Beat%20In%20Bar) 4) 1))))

(event EventOnUserRemoved
  (params (User :type Playspace%20User%20Structure%20%28by%20ref%29) (PlayerState :type Player%20State%20Object%20Reference))
  (bind (_userid _leafplayspace _playerstatecached) (Utilities|Struct|BreakPlayspaceUser User))
  ;[0.9454 0.3569 1] Players added and removed to the playspace
  (Development|PrintString (Utilities|String|Append (PlayerState|GetPlayerName _playerstatecached) " Left" (Utilities|GetDisplayName self)) false true "(R=1.000000,G=0.276000,B=0.000000,A=1.000000)" 1e+01 ""))

(event Custom|OnTempoChanged
  (bind _musicmanager (Variables|JamPlayspace|GetMusicManager))
  (bind (_outmusickey _outmusickeymode _outtempo _returnvalue) (Mojo|JamSession|TryGetKeyModeTempo _musicmanager))
  ;[1 0.573 0.4931] Reactive VFX
  (if _returnvalue
    (Class|BPJamPlayspaceVolume|OnTempoChanged (Variables|Default|GetPlayspaceVolume) _outtempo)
    (else
      (Development|PrintString "Could not get audio state!" false true "(R=1.000000,G=0.000000,B=0.000000,A=1.000000)" 1e+01 ""))))

(event Custom|ListenToLoopEvents
  ;[0.3687 0.6041 1] Loop Events
  (Networking|CanExecuteCosmeticEvents
    (:True
      (Utilities|FlowControl|Sequence
        (:then_0
          ;;[0.3281 1 0.9202] A Loop Stopped Event
          (bind (_asynctaskproxy _payload) (GameplayEventRouter|StartListeningToEvent (CallFunction|GetOwnedEventRouter) "/Script/FMJamSessionRuntime.JamEvent_JamLoopStopped")
            (:then
              (Variables|Default|SetJamLoopStoppedAsyncTaskRef _asynctaskproxy))
            (:OnEventReceived
              (CallFunction|OnJamLoopStoppedEvent _payload))))
        (:then_1
          ;;[0.003922 0.3725 0.8314] A Loop Started Event
          (bind (_asynctaskproxy_2 _payload_2) (GameplayEventRouter|StartListeningToEvent (CallFunction|GetOwnedEventRouter) "/Script/FMJamSessionRuntime.JamEvent_JamLoopStarted")
            (:then
              (Variables|Default|SetJamLoopStartedAsyncTaskRef _asynctaskproxy_2))
            (:OnEventReceived
              (CallFunction|OnJamLoopStartedEvent _payload_2))))
        (:then_2
          ;;[0.7556 0.5311 1] Tempo Events for VFX
          (bind (_asynctaskproxy_3 _payload_3) (GameplayEventRouter|StartListeningToEvent (CallFunction|GetOwnedEventRouter) "/Script/SparksCoreRuntime.SparksCoreAudioEvent_OnBeat")
            (:OnEventReceived
              (CallFunction|OnBeat (Utilities|Struct|BreakSparksCoreAudioEventonBeat _payload_3) (.BeatInBar (Utilities|Struct|BreakSparksCoreAudioEventonBeat _payload_3))))))
        (:then_3
          (GameplayEventRouter|StartListeningToEvent (CallFunction|GetOwnedEventRouter) "/Script/SparksCoreRuntime.SparksCoreAudioEvent_TempoChanged"
            (:then
              _)
            (:OnEventReceived
              _))
          (CallFunction|OnTempoChanged))))))

(event EventBeginPlay
  ;[0.6133 1 0.5993] BeginPlay / EndPlay
  (Utilities|FlowControl|Sequence
    (:then_0
      (|CallParentFunction))
    (:then_1
      (Networking|SwitchHasAuthority
        (:Authority
          (Class|PlayspaceComponentSparksPerfMem|StartTrackingServerPerf (Variables|SparksMusicPlayspace|GetPerfMemComp) (Utilities|Struct|MakeSparksServerPerfMemSettings "Jam")))
        (:Remote
          (CallFunction|SetMetasoundAttenuation)))
      (CallFunction|ListentoLoopEvents)
      (CallFunction|InitPlayspaceVolume)
      (for _array_element (Mojo|JamSession|GetAllMusicSlots (Variables|JamPlayspace|GetMusicManager))
        (if (Class|JamMusicSlot|IsinUse _array_element)
          (CallFunction|OnJamLoopStartedEvent (Class|JamMusicSlot|GetParamsasJamEvent _array_element)))))))

(event EventEndPlay
  (params (EndPlayReason :type EEndPlayReason%20Enum :default "Destroyed"))
  ;[0.6133 1 0.5993] BeginPlay / EndPlay
  (Utilities|FlowControl|Sequence
    (:then_0
      (GameplayEventRouter|StopListeningtoEvent (Variables|Default|GetJamLoopStoppedAsyncTaskRef))
      (Variables|Default|SetJamLoopStoppedAsyncTaskRef 0)
      (GameplayEventRouter|StopListeningtoEvent (Variables|Default|GetJamLoopStartedAsyncTaskRef))
      (Variables|Default|SetJamLoopStartedAsyncTaskRef 0)
      (GameplayEventRouter|StopListeningtoEvent (Variables|Default|GetBeatAsyncTaskRef))
      (Variables|Default|SetBeatAsyncTaskRef 0)
      (Networking|CanExecuteCosmeticEvents
        (:True
          (Class|JamSessionComponentReactiveFX|ClearReactiveEffects (Variables|JamPlayspace|GetReactiveFXComponent) true))
        (:False
          _))
      (Networking|SwitchHasAuthority
        (:Authority
          (Class|PlayspaceComponentSparksPerfMem|StopTrackingServerPerf (Variables|SparksMusicPlayspace|GetPerfMemComp)))))
    (:then_1
      (|CallParentFunction EndPlayReason))))

(event EventBeginShutdownJam
  ;[1 0 0.4842] Kill the Playspace via the volume
  (Networking|SwitchHasAuthority
    (:Authority
      (if (not (Class|BPJamPlayspaceVolume|IsPersistent (Variables|Default|GetPlayspaceVolume)))
        (bind _playspacevolume (Variables|Default|GetPlayspaceVolume)
          (:then
            (Development|PrintString "Kicking all players and ending Jam" false)
            (Class|BPJamPlayspaceVolume|KickAllPlayersandEndJam 0))
          (:else
            (Fortnite|FortShippingLog 0 "OnJamLoopStopped: Could not find JamPlayspaceVolume" true)))))))

(fn DoesPlayspaceBlockJamming
  (bind _asbp_jam_playspace_volume (Utilities|Casting|CastToBP_JamPlayspace_Volume (Variables|Playspace|GetBoundGameplayVolume))
    (:then
      (bind _blockjam (Class|BPJamPlayspaceVolume|ShouldBlockJam _asbp_jam_playspace_volume))
      (return _blockjam))
    (:CastFailed
      (return false))))

(fn GetJamAudioSource
  (bind _playspacevolume (Variables|Default|GetPlayspaceVolume)
    (:then
      (return (Variables|/FMJamSystem/Playspace/BP_JamPlayspace_Volume.BP_JamPlayspace_Volume_C|GetJamSource _playspacevolume)))
    (:else
      (CallFunction|InitPlayspaceVolume)
      (bind _playspacevolume_2 (Variables|Default|GetPlayspaceVolume)
        (:then
          (return (Variables|/FMJamSystem/Playspace/BP_JamPlayspace_Volume.BP_JamPlayspace_Volume_C|GetJamSource _playspacevolume_2)))
        (:else
          (return 0))))))

(fn InitPlayspaceVolume
  (bind _boundgameplayvolume (Variables|Playspace|GetBoundGameplayVolume)
    (:then
      (bind _asbp_jam_playspace_volume (Utilities|Casting|CastToBP_JamPlayspace_Volume _boundgameplayvolume)
        (:then
          (Variables|Default|SetPlayspaceVolume _asbp_jam_playspace_volume)
          (CallFunction|SetMetasoundAttenuation)
          (CallFunction|OnTempoChanged))
        (:CastFailed
          (Fortnite|FortShippingLog self (Utilities|String|Append (LocalVariables|GetFunction%20Name) ": Cannot Cast to BP_JamPlayspace_Volume") true))))
    (:else
      (Fortnite|FortShippingLog self (Utilities|String|Append (LocalVariables|GetFunction%20Name) ": Bound Gameplay Volume Invalid") true))))

(fn LogMetasoundPlayerLocation
  (params (CalledFromString :type String))
  (bind _boundgameplayvolume (Variables|Playspace|GetBoundGameplayVolume))
  (Fortnite|FortShippingLog self (Utilities|String|Append "Logging Jam location date from " CalledFromString))
  (Fortnite|FortShippingLog self (Utilities|String|Append "Jam Playspace Metasound Player Location: " (Utilities|String|ToString(Vector) (Transformation|GetWorldLocation (Variables|SparksMusicPlayspace|GetMetasoundPlayer)))))
  (Utilities|IsValid _boundgameplayvolume
    (:Is%20Valid
      (Fortnite|FortShippingLog self (Utilities|String|Append "Jam Playspace Location: " (Utilities|String|ToString(Vector) (Transformation|GetActorLocation _boundgameplayvolume)))))))

(fn OnJamLoopStartedEvent
  (params (Payload :type Jam%20Event%20Jam%20Loop%20Started%20Structure))
  (bind (_songshortname _catalogentry _looptype _playernetid _loopinstanceid _bisautojammer) (Utilities|Struct|BreakJamEventJamLoopBase Payload))
  (Networking|CanExecuteCosmeticEvents
    (:True
      (if (or (Online|UniqueNetIdRepl|IsValid(UniqueIdRepl) _playernetid) _bisautojammer)
        (Utilities|IsValid (Variables|Default|GetPlayspaceVolume)
          (:Is%20Valid
            (CallFunction|SetMetasoundAttenuation))
          (:Is%20Not%20Valid
            (CallFunction|InitPlayspaceVolume)))))
    (:False
      ;[0.3698 0.004154 0] Server does nothing
      (return))))

(fn OnJamLoopStoppedEvent
  (params (Payload :type Jam%20Event%20Jam%20Loop%20Stopped%20Structure)))

(fn SetMetasoundAttenuation)

(fn SwitchStutterFX
  (params (SwitchAudioFX :type Integer :default 1))
  ;[1 1 1] Use "Ke*" before the console command!
  ;
  ; Param Input is disconnected in the MSS for now.  Its Rate is chosen randomly at each FX Active Toggle.
  ;
  ; This function added for when we want a Param input for the effect.
  (Audio|Parameter|SetIntegerParameter (Variables|SparksMusicPlayspace|GetMetasoundPlayer) "J0-FX0-Param" SwitchAudioFX :!class /Script/Engine.AudioComponent)
  (Development|PrintString (Utilities|String|Append "Jam FX Settings" (Utilities|String|ToString(Integer) SwitchAudioFX))))

(fn ToggleStutterFX
  (params (ToggleStutterFX :type Boolean))
  ;[1 1 1] Use "Ke*" before the console command!
  (Audio|Parameter|SetBooleanParameter (Variables|SparksMusicPlayspace|GetMetasoundPlayer) "J0-FX0-IsActive" ToggleStutterFX :!class /Script/Engine.AudioComponent)
  (Development|PrintString (Utilities|String|Append "Jam FX Active" (Utilities|String|ToString(Boolean) ToggleStutterFX))))

(fn ConstructionScript)
