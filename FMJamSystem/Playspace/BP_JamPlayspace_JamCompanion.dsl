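; BP_JamPlayspace_JamCompanion is a Jam playspace Blueprint whose only working logic is debug output for network dormancy changes.
; When its dormancy manager component reports a change, it prints the old and new ENetDormancy states to screen and log under a fixed debug key.
; BeginPlay, Tick and the construction script add no behavior of their own.
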(event EventBeginPlay
  (|CallParentFunction) ; This node is disabled and will not be called.\nDrag off pins to build functionality.
)

(event EventTick
  (params (DeltaSeconds :type Float%20%28single-precision%29)))

(event OnPlayspaceDormancyChangedEvent(PlayspaceComponent_DormancyManager)
  (params (AffectedPlayspace :type Playspace%20Object%20Reference) (OldNetDormancy :type ENetDormancy%20Enum :default "DORM_Never") (NewNetDormancy :type ENetDormancy%20Enum :default "DORM_Never") (DormancyManager :type Playspace%20Component%20Dormancy%20Manager%20Object%20Reference))
  (Development|PrintString (Utilities|String|ToString(Text) (Utilities|Text|FormatText "INVTEXT(\"Changing from {Old} to {New} Dormancy\")" (Utilities|String|EnumtoString OldNetDormancy) (Utilities|String|EnumtoString NewNetDormancy))) true true "(R=0.803273,G=0.224325,B=1.000000,A=1.000000)" 25.0 "JamCompanion_DormancyDebug"))

(fn ConstructionScript
  (|CallParentFunction))
