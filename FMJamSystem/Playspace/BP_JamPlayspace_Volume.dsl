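; Jam playspace volume that tracks player pawns in a jam circle, with overridable collider and persistence functions.
; On the server, overlaps grant or remove the jam gameplay effect, and non-persistent volumes shut down when empty, delaying destruction to avoid a double-removal crash.
; Clients warn entering players about licensed-audio restrictions.
; Inner-collider overlaps set server/client membership and quick-join tags, while audio virtualization toggles submix registration for audio reactivity.
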(event EventBeginPlay
  (bind _collider (CallFunction|GetInnerCollider))
  (Utilities|FlowControl|Sequence
    (:then_0
      (Networking|SwitchHasAuthority
        (:Authority
          (Collision|BindEventtoOnActorBeginOverlap (EventDispatchers|CreateEvent :!event Custom|ActorEnter))
          (Collision|BindEventtoOnActorEndOverlap (EventDispatchers|CreateEvent :!event Custom|ActorLeave))
          ;[0.1356 0.4072 0.3763] Get any players that are already in the playspace immediately after it is set up
          (Utilities|FlowControl|DelayUntilNextTick)
          (for _array_element (Collision|GetOverlappingActors (Variables|Default|GetOverlap) "/Script/FortniteGame.FortPlayerPawn" :!class /Script/Engine.PrimitiveComponent)
            (CallFunction|ActorEnter 0 _array_element)))
        (:Remote
          ;[1 1 1] we need this on the client as well to warn about licensed music
          (Collision|BindEventtoOnActorBeginOverlap (EventDispatchers|CreateEvent :!event Custom|ActorEnter)))))
    (:then_1
      ;[0 1 0.6062] Listen for actors entering/leaving the inner collider (both server and client)
      (Utilities|IsValid _collider
        (:Is%20Valid
          (Collision|BindEventtoOnComponentBeginOverlap _collider (delegate Custom|OnBeginOverlapWithInnerCollider))
          (Collision|BindEventtoOnComponentEndOverlap (CallFunction|GetInnerCollider) (delegate Custom|OnEndOverlapWithInnerCollider)))))))

(event Custom|OnBeginOverlapWithInnerCollider
  (params (OverlappedComponent :type Primitive%20Component%20Object%20Reference) (OtherActor :type Actor%20Object%20Reference) (OtherComp :type Primitive%20Component%20Object%20Reference) (OtherBodyIndex :type Integer) (bFromSweep :type Boolean) (SweepResult :type Hit%20Result%20Structure%20%28by%20ref%29))
  ;[0 1 0.6062] Listen for actors entering/leaving the inner collider (both server and client)
  (CallFunction|OnInnerColliderOverlap OtherActor true))

(event Custom|OnEndOverlapWithInnerCollider
  (params (OverlappedComponent :type Primitive%20Component%20Object%20Reference) (OtherActor :type Actor%20Object%20Reference) (OtherComp :type Primitive%20Component%20Object%20Reference) (OtherBodyIndex :type Integer))
  ;[0 1 0.6062] Listen for actors entering/leaving the inner collider (both server and client)
  (CallFunction|OnInnerColliderOverlap OtherActor))

(event Custom|ActorEnter
  (params (OverlappedActor :type Actor%20Object%20Reference) (OtherActor :type Actor%20Object%20Reference))
  (Networking|SwitchHasAuthority
    (:Authority
      (bind _asfort_player_pawn (Utilities|Casting|CastToFortPlayerPawn OtherActor))
      (CallFunction|ApplyJamGE _asfort_player_pawn)
      (if (< (Utilities|Array|FindItem (Variables|Default|GetPlayersInVolume) _asfort_player_pawn) 0)
        (Utilities|Array|AddUnique (Variables|Default|GetPlayersInVolume) _asfort_player_pawn)))
    (:Remote
      ;[1 1 1] if they try to play and licensed music is disabled, warn them
      (bind _asfort_player_pawn_2 (Utilities|Casting|CastToFortPlayerPawn OtherActor))
      (bind _returnvalue (Fortnite|Audio|GetLicensedAudioTreatment))
      (if (and (Utilities|Enum|NotEqual(Enum) _returnvalue "None") (> (Fortnite|GetNumericCVar "FMJam.RespectLicensedMusicSetting") 0))
        (CallFunction|PlayJamErrorMessageClient _asfort_player_pawn_2 true "(GameplayTags=((TagName=\"FMJam.Playspace.LicensedAudioDisabled\")))")))))

(event Custom|ActorLeave
  (params (OverlappedActor :type Actor%20Object%20Reference) (OtherActor :type Actor%20Object%20Reference))
  (Networking|SwitchHasAuthority
    (:Authority
      (Utilities|Casting|CastToFortPlayerPawn OtherActor)
      (bind _wasremoved (CallFunction|RemovePlayer OtherActor))
      (if _wasremoved
        (if (Utilities|Array|IsEmpty (Variables|Default|GetPlayersInVolume))
          (CallFunction|ShutDownJamCircle))))))

(event Custom|KickAllPlayersAndEndJam
  (Networking|SwitchHasAuthority
    (:Authority
      (if (not (CallFunction|IsPersistent))
        (for _array_element (Variables|Default|GetPlayersInVolume)
          (CallFunction|ActorLeave 0 _array_element))
        (CallFunction|ShutDownJamCircle)))))

(event Custom|ShutDownJamCircle
  (Networking|SwitchHasAuthority
    (:Authority
      (if (CallFunction|IsPersistent)
        _
        (elif (not (Variables|Default|GetShuttingDown))
          ;[1 1 1] Don't attempt to shut down multiple times
          (Variables|Default|SetShuttingDown true)
          (Collision|SetActorEnableCollisionwithChildren 0)
          (Rendering|SetActorHiddenInGame true)
          (Utilities|FlowControl|Sequence
            (:then_0
              (if (> (Fortnite|GetNumericCVar "FMJam.ForceExitAllPlayersDuringShutdown") 0)
                (bind _asjam_playspace (Utilities|Casting|CastToJamPlayspace (CallFunction|GetPlayspace)))
                (for _array_element (Class|SparksMusicPlayspace|GetAllPlayerControllers _asjam_playspace)
                  (Class|SparksMusicPlayspace|RemovePlayspaceUser _asjam_playspace _array_element))))
            (:then_1
              ;[1 1 1] We must delay destruction of the playspace, because we got here possibly due to a player leaving (but they haven't really been removed yet). This delay avoid a crash bug of trying to remove the player twice.
              (Utilities|FlowControl|Delay (Variables|Default|GetPlayspaceDestructionDelay))
              (Actor|DestroyActor))))))))

(event Custom|ApplyJamGE
  (params (Fort%20Player%20Pawn :type Fort%20Player%20Pawn%20Object%20Reference))
  (Fortnite|ApplyGameplayEffectsToActor Fort%20Player%20Pawn (Utilities|Array|MakeArray (Variables|Default|GetJamGE)) (Transformation|GetActorLocation Fort%20Player%20Pawn) false 0 (Ability|GetAbilitySystemComponent Fort%20Player%20Pawn)))

(event Custom|RemoveJamGE
  (params (FortPlayerPawn :type Fort%20Player%20Pawn%20Object%20Reference))
  (GameplayEffects|RemoveActiveGameplayEffectbySourceEffect (Ability|GetAbilitySystemComponent FortPlayerPawn) (Variables|Default|GetJamGE) 0 -1))

(event Custom|UpdateAmplitude
  (params (Amplitude :type Float%20%28double-precision%29)))

(event Custom|OnBeat
  (params (Beat%20in%20Bar%20%281-4%29 :type Integer) (Measure%20Number%20%281-32%29 :type Integer))
  (Default|CallOnBeatCalled Beat%20in%20Bar%20%281-4%29 Measure%20Number%20%281-32%29))

(event Custom|OnTempoChanged
  (params (NewTempo :type Integer)))

(event OnAudioVirtualizationChanged(JamSource)
  (params (bIsVirtualized :type Boolean))
  ;[0 0 0] Start/Stop Audio Reactivity
  (if bIsVirtualized
    (AudioAnalysis|UnRegisterDefaultSubmix (WorldSubsystems|GetFortAudioAnalysisSubsystem))
    (else
      (AudioAnalysis|RegisterDefaultSubmix (WorldSubsystems|GetFortAudioAnalysisSubsystem)))))

(event EventEndPlay
  (params (EndPlayReason :type EEndPlayReason%20Enum :default "Destroyed"))
  ;[0 1 0.6062] Make sure that this even gets removed when we die.  It will also get removed when we deactivate the sound source
  (Audio|RemoveEvent (GameInstanceSubsystems|GetMusicEventSubsystem) (Variables|Default|GetJamSource)))

(event Custom|PlayJamErrorMessage_Client
  (params (Player%20Pawn :type Fort%20Player%20Pawn%20Object%20Reference) (Play%20Montage :type Boolean) (Failed%20Reason :type Gameplay%20Tag%20Container%20Structure))
  (FailedAction|UnabletoPerformAction Player%20Pawn Play%20Montage Failed%20Reason))

(fn AddOrRemoveLooseGameplayTag
  (params (Actor :type Actor%20Object%20Reference) (GameplayTag :type Gameplay%20Tag%20Structure) (Add :type Boolean))
  (bind _returnvalue (Ability|GetAbilitySystemComponent Actor))
  (bind _gameplaytag (Variables|GetGameplayTag))
  (bind _returnvalue_2 (GameplayTags|HasMatchingGameplayTag _returnvalue _gameplaytag))
  (Utilities|IsValid _returnvalue
    (:Is%20Valid
      (if (Variables|GetAdd)
        (if (not _returnvalue_2)
          (Ability|GameplayEffect|AddLooseGameplayTags (Variables|GetActor) (GameplayTags|MakeGameplayTagContainerfromTag (Variables|GetGameplayTag))))
        (else
          (while _returnvalue_2
            (Ability|GameplayEffect|RemoveLooseGameplayTags (Variables|GetActor) (GameplayTags|MakeGameplayTagContainerfromTag (Variables|GetGameplayTag)))))))))

(fn BP_IsPersistent
  (return (CallFunction|IsPersistent)))

(fn GetCenterLocation
  (return (Transformation|GetActorLocation)))

(fn GetInnerCollider
  (return 0))

(fn GetBlockedJamMessage
  (return "(GameplayTags=((TagName=\"FMJam.Playspace.JammingBlocked\")))"))

(fn GetInnerCollider
  (return (CallFunction|GetInnerCollider)))

(fn GetOuterCollider
  (return 0))

(fn IsPersistent
  (return false))

(fn OnInnerColliderOverlap
  (params (OtherActor :type Actor%20Object%20Reference) (bInCollider :type Boolean))
  (Utilities|FlowControl|Sequence
    (:then_0
      (Networking|SwitchHasAuthority
        (:Authority
          (LocalVariables|SetInnerCircleTag "(TagName=\"FMJam.Session.InnerCollider.Member.Server\")"))
        (:Remote
          (LocalVariables|SetInnerCircleTag "(TagName=\"FMJam.Session.InnerCollider.Member.Client\")")))
      (CallFunction|AddorRemoveLooseGameplayTag (Variables|GetOtherActor) (LocalVariables|GetInnerCircleTag) (Variables|GetbInCollider)))
    (:then_1
      (bind _returnvalue (Development|GetConsoleVariableBoolValue "FMJam.QuickJoinRequiresTargeting"))
      (if (not _returnvalue)
        (CallFunction|AddorRemoveLooseGameplayTag (Variables|GetOtherActor) "(TagName=\"FMJam.Jam.QuickJoinTouchControlAvailable\")" (Variables|GetbInCollider))))
    (:then_2
      (return))))

(fn RemovePlayer
  (params (PlayerActor :type Actor%20Object%20Reference))
  (bind _asfort_player_pawn (Utilities|Casting|CastToFortPlayerPawn PlayerActor)
    (:then
      (CallFunction|RemoveJamGE _asfort_player_pawn)
      (bind _returnvalue (Utilities|Array|RemoveItem (Variables|Default|GetPlayersInVolume) _asfort_player_pawn))
      (return _returnvalue))
    (:CastFailed
      (return false))))

(fn ShouldBlockJam
  (return (Variables|Jam|GetShouldBlockJamming)))

(fn ConstructionScript
  (Variables|/Script/Engine.AudioComponent|SetAttenuationSettings (Variables|Default|GetJamSourceAttenuation) (Variables|Default|GetJamSource)))
