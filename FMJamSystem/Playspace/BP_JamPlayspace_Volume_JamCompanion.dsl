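; Jam Companion playspace volume that exposes colliders, center and per-loop-type ring transforms for jam sessions.
; Once its playspace is set, it registers loop start/stop listeners on that playspace's event router and refreshes the audio cue.
; The cosmetic-only cue fades in when no slots are active and the companion isn't set to play alone, otherwise fading out with a loop-start sound.
; Beat and tempo events drive the Niagara border.
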(event EventBeginPlay
  (|CallParentFunction)
  (CallFunction|OnPlayingLoopChanged))

(event EventEndPlay
  (params (EndPlayReason :type EEndPlayReason%20Enum :default "Destroyed"))
  (CallFunction|CleanupListeners)
  (|CallParentFunction EndPlayReason))

(event EventOnBeat
  (params (Beat%20in%20Bar%20%281-4%29 :type Integer) (Measure%20Number%20%281-32%29 :type Integer))
  (|CallParentFunction Beat%20in%20Bar%20%281-4%29 Measure%20Number%20%281-32%29)
  (Niagara|SetNiagaraVariable(Float) (Variables|Default|GetNS_JamSpace_Border_SquareInner) "GameTime" (Utilities|Time|GetGameTimeinSeconds)))

(event EventOnTempoChanged
  (params (NewTempo :type Integer))
  (Niagara|SetNiagaraVariable(Int32) (Variables|Default|GetNS_JamSpace_Border_SquareInner) "Tempo" NewTempo))

(event Custom|PlayspaceSet
  (params (SelfVolume :type Gameplay%20Volume%20Object%20Reference) (NewPlayspace :type Playspace%20Object%20Reference))
  (CallFunction|SetupListeners))

(event Custom|CleanupListeners
  (Utilities|FlowControl|Sequence
    (:then_0
      (GameplayEventRouter|StopListeningtoEvent (Variables|Default|GetJamLoopStarted_EventListenerHandle)))
    (:then_1
      (GameplayEventRouter|StopListeningtoEvent (Variables|Default|GetJamLoopStopped_EventListenerHandle)))))

(event Custom|SetupListeners
  (CallFunction|CleanupListeners)
  (bind _asbp_jam_playspace_jam_companion (Utilities|Casting|CastToBP_JamPlayspace_JamCompanion (CallFunction|GetPlayspace)))
  (Utilities|IsValid (Class|SparksMusicPlayspace|GetOwnedEventRouter _asbp_jam_playspace_jam_companion)
    (:Is%20Valid
      (Utilities|FlowControl|Sequence
        (:then_0
          (bind _asynctaskproxy (GameplayEventRouter|StartListeningToEvent (Class|SparksMusicPlayspace|GetOwnedEventRouter _asbp_jam_playspace_jam_companion) "/Script/FMJamSessionRuntime.JamEvent_JamLoopStarted")
            (:then
              (Variables|Default|SetJamLoopStarted_EventListenerHandle _asynctaskproxy))
            (:OnEventReceived
              _)))
        (:then_1
          (bind _asynctaskproxy_2 (GameplayEventRouter|StartListeningToEvent (Class|SparksMusicPlayspace|GetOwnedEventRouter _asbp_jam_playspace_jam_companion) "/Script/FMJamSessionRuntime.JamEvent_JamLoopStopped")
            (:then
              (Variables|Default|SetJamLoopStopped_EventListenerHandle _asynctaskproxy_2))
            (:OnEventReceived
              _))))
      (CallFunction|OnPlayingLoopChanged))))

(fn GetCenterLocation
  (return (Transformation|GetWorldLocation (Variables|Default|GetInner%20Collider))))

(fn GetInnerCollider
  (return (Variables|Default|GetInner%20Collider)))

(fn GetJamRingTransformForTrackType
  (params (LoopType :type EFMJamLoopType%20Enum :default "Lead"))
  (bind _drumjamring (Variables|Default|GetDrumJamRing))
  (bind _bassjamring (Variables|Default|GetBassJamRing))
  (bind _leadjamring (Variables|Default|GetLeadJamRing))
  (bind _vocalsjamring (Variables|Default|GetVocalsJamRing))
  (bind _returnvalue (Utilities|Select :Index LoopType :Lead _vocalsjamring :Misc _leadjamring :Bass _bassjamring :Beat _drumjamring))
  (bind _returnvalue_2 (Transformation|GetWorldTransform _returnvalue))
  (Utilities|IsValid _returnvalue
    (:Is%20Valid
      (Development|PrintString (Utilities|String|Append "Spawn " (Utilities|String|EnumtoString (Variables|GetLoopType)) " Ring at " (Utilities|String|ToString(Transform) _returnvalue_2)) true false "(R=0.000000,G=1.000000,B=0.303995,A=1.000000)" 45.0 (Utilities|String|StringToName (Utilities|String|EnumtoString (Variables|GetLoopType))))
      (return _returnvalue_2))
    (:Is%20Not%20Valid
      (Development|PrintString "WARN: Invalid ring type" true false "(R=1.000000,G=0.266087,B=0.345836,A=1.000000)" 15.0)
      (return (Transformation|GetActorTransform)))))

(fn GetOuterCollider
  (return (Variables|Default|GetCollider)))

(fn IsPersistent
  (return true))

(fn OnPlayingLoopChanged
  (Networking|CanExecuteCosmeticEvents
    (:True
      (bind _asbp_jam_playspace_jam_companion (Utilities|Casting|CastToBP_JamPlayspace_JamCompanion (CallFunction|GetPlayspace)))
      (CallFunction|SetAudioCueActive (and (== (Mojo|JamSession|CountActiveSlotswithParams (Class|JamSessionInterface|GetMusicManager _asbp_jam_playspace_jam_companion) (Utilities|Struct|MakeJamMusicSlotSearchParams "INVALID" "Num" "IncludeAll")) 0) (not (JamSystem|GetShouldPlaySongAlone (Variables|/FMJamSystem/Playspace/BP_JamPlayspace_JamCompanion.BP_JamPlayspace_JamCompanion_C|GetJamCompanion _asbp_jam_playspace_jam_companion)))))
      (return))
    (:False
      (return))))

(fn SetAudioCueActive
  (params (bShouldAudioCueBeAudible :type Boolean))
  (Networking|CanExecuteCosmeticEvents
    (:True
      (if bShouldAudioCueBeAudible
        (Audio|Components|Audio|FadeIn (Variables|Default|GetJamCompanionAudioCue) 2.0 1.0 0.0 "Linear" :!class /Script/Engine.AudioComponent)
        (Development|PrintString (Utilities|String|Append "JamVolume Audio cue is " (Utilities|String|SelectString "Audible" "Silent" bShouldAudioCueBeAudible)) true false "(R=0.182707,G=1.000000,B=0.151953,A=1.000000)" 15.0 "JamCompanionAudioCue")
        (return)
        (else
          (Audio|Components|Audio|FadeOut (Variables|Default|GetJamCompanionAudioCue) 2.0 0.0 "Linear" :!class /Script/Engine.AudioComponent)
          (Audio|SpawnSoundAttached "/FMJamSystem/Audio/SC_JamLoop_Start.SC_JamLoop_Start" (Variables|Transformation|GetRootComponent)))))))

(fn ConstructionScript
  (|CallParentFunction)
  (EventDispatchers|BindEventtoPlayspaceSetEvent (EventDispatchers|CreateEvent :!event PlayspaceSet)))
