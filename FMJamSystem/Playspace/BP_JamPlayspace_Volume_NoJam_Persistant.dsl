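; Persistent playspace volume Blueprint for the FMJam system that marks an area where jamming is blocked.
; It overrides playspace queries to report itself as persistent and supply the JammingBlocked gameplay tag as its blocked-jam message.
; A single collider serves as the outer collider and gives the center location, while no inner collider is provided.
; The construction script only calls its parent.
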(fn GetCenterLocation
  (return (Transformation|GetWorldLocation (Variables|Default|GetCollider))))

(fn GetInnerCollider
  (return 0))

(fn GetBlockedJamMessage
  (return "(GameplayTags=((TagName=\"FMJam.Playspace.JammingBlocked\")))"))

(fn GetOuterCollider
  (return (Variables|Default|GetCollider)))

(fn IsPersistent
  (return true))

(fn ConstructionScript
  (|CallParentFunction))
