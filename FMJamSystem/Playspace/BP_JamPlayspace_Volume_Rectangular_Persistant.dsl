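; Rectangular, persistent variant of the Jam playspace volume, a child Blueprint that reports itself as persistent.
; It supplies its inner collider, outer collider and center (the inner collider's location) through functions.
; On authority, an actor leave event casts to FortPlayerPawn and removes that player.
; Beat and tempo events push game time and tempo into the inner square border Niagara effect, keeping the visual informed of the music timing.
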(event EventActorLeave
  (params (OverlappedActor :type Actor%20Object%20Reference) (OtherActor :type Actor%20Object%20Reference))
  (Networking|SwitchHasAuthority
    (:Authority
      (Utilities|Casting|CastToFortPlayerPawn OtherActor)
      (CallFunction|RemovePlayer OtherActor))))

(event EventOnBeat
  (params (Beat%20in%20Bar%20%281-4%29 :type Integer) (Measure%20Number%20%281-32%29 :type Integer))
  (|CallParentFunction Beat%20in%20Bar%20%281-4%29 Measure%20Number%20%281-32%29)
  (Niagara|SetNiagaraVariable(Float) (Variables|Default|GetNS_JamSpace_Border_SquareInner) "GameTime" (Utilities|Time|GetGameTimeinSeconds)))

(event EventOnTempoChanged
  (params (NewTempo :type Integer))
  (Niagara|SetNiagaraVariable(Int32) (Variables|Default|GetNS_JamSpace_Border_SquareInner) "Tempo" NewTempo))

(fn GetCenterLocation
  (return (Transformation|GetWorldLocation (Variables|Default|GetInner%20Collider))))

(fn GetInnerCollider
  (return (Variables|Default|GetInner%20Collider)))

(fn GetOuterCollider
  (return (Variables|Default|GetCollider)))

(fn IsPersistent
  (return true))

(fn ConstructionScript
  (|CallParentFunction))
