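; Sphere-shaped jam playspace volume that specializes a parent playspace class with spherical colliders and a beat-aware Niagara border.
; It overrides the center and collider getters to return the Inner Sphere's location and the Inner and Outer Sphere components.
; On each beat it calls the parent, then writes current game time to the inner circle border effect.
; Tempo changes push the new tempo to that same effect.
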(event EventOnBeat
  (params (Beat%20in%20Bar%20%281-4%29 :type Integer) (Measure%20Number%20%281-32%29 :type Integer))
  (|CallParentFunction Beat%20in%20Bar%20%281-4%29 Measure%20Number%20%281-32%29)
  (Niagara|SetNiagaraVariable(Float) (Variables|Default|GetNS_JamSpace_Border_CircleInner) "GameTime" (Utilities|Time|GetGameTimeinSeconds)))

(event EventOnTempoChanged
  (params (NewTempo :type Integer))
  (Niagara|SetNiagaraVariable(Int32) (Variables|Default|GetNS_JamSpace_Border_CircleInner) "Tempo" NewTempo))

(fn GetCenterLocation
  (return (Transformation|GetWorldLocation (Variables|Default|GetInner%20Sphere))))

(fn GetInnerCollider
  (return (Variables|Default|GetInner%20Sphere)))

(fn GetOuterCollider
  (return (Variables|Default|GetOuter%20Sphere)))

(fn ConstructionScript
  (|CallParentFunction))
