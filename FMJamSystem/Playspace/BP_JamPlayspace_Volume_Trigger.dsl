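; Persistent trigger volume for a Jam playspace that relays player presence and music timing through event dispatchers and drives the square border mesh's material.
; Inner collider overlaps broadcast entering and leaving, while the authority removes departing player pawns.
; Beat events write truncated game time, tempo and the music manager's active slot count into custom primitive data, then broadcast the beat.
; Tempo changes are stored and rebroadcast.
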(event OnComponentBeginOverlap(InnerCollider)
  (params (OverlappedComponent :type Primitive%20Component%20Object%20Reference) (OtherActor :type Actor%20Object%20Reference) (OtherComp :type Primitive%20Component%20Object%20Reference) (OtherBodyIndex :type Integer) (bFromSweep :type Boolean) (SweepResult :type Hit%20Result%20Structure%20%28by%20ref%29))
  (Default|CallOnActorEntering))

(event OnComponentEndOverlap(InnerCollider)
  (params (OverlappedComponent :type Primitive%20Component%20Object%20Reference) (OtherActor :type Actor%20Object%20Reference) (OtherComp :type Primitive%20Component%20Object%20Reference) (OtherBodyIndex :type Integer))
  (bind _jammusicslotsearchparams (Utilities|Struct|MakeJamMusicSlotSearchParams "INVALID" "Num" "IncludeAll"))
  (bind _asbp_jam_playspace (Utilities|Casting|CastToBP_JamPlayspace (CallFunction|GetPlayspace self)))
  (Rendering|Material|SetCustomPrimitiveDataFloat (Variables|Default|GetSM_JamPlayspace_Border_Square) 3 (Math|Conversions|ToFloat(Integer) (Mojo|JamSession|CountActiveSlotswithParams (Class|JamSessionInterface|GetMusicManager _asbp_jam_playspace) _jammusicslotsearchparams)))
  (Default|CallOnActorLeaving (Mojo|JamSession|CountActiveSlotswithParams (Class|JamSessionInterface|GetMusicManager _asbp_jam_playspace) _jammusicslotsearchparams)))

(event EventActorEnter
  (params (OverlappedActor :type Actor%20Object%20Reference) (OtherActor :type Actor%20Object%20Reference)))

(event EventActorLeave
  (params (OverlappedActor :type Actor%20Object%20Reference) (OtherActor :type Actor%20Object%20Reference))
  (Networking|SwitchHasAuthority
    (:Authority
      (Utilities|Casting|CastToFortPlayerPawn OtherActor)
      (CallFunction|RemovePlayer OtherActor))))

(event EventOnBeat
  (params (Beat%20in%20Bar%20%281-4%29 :type Integer) (Measure%20Number%20%281-32%29 :type Integer))
  (bind _returnvalue (Utilities|Time|GetGameTimeinSeconds))
  (bind _returnvalue_2 (Math|Float|Truncate _returnvalue))
  (bind _jammusicslotsearchparams (Utilities|Struct|MakeJamMusicSlotSearchParams "INVALID" "Num" "IncludeAll"))
  (bind _sm_jamplayspace_border_square (Variables|Default|GetSM_JamPlayspace_Border_Square))
  (|CallParentFunction Beat%20in%20Bar%20%281-4%29 Measure%20Number%20%281-32%29)
  (Variables|Default|SetMeasure Measure%20Number%20%281-32%29)
  (bind _asbp_jam_playspace (Utilities|Casting|CastToBP_JamPlayspace (CallFunction|GetPlayspace self)))
  (Variables|Default|SetBeats _returnvalue_2)
  (Rendering|Material|SetCustomPrimitiveDataFloat _sm_jamplayspace_border_square 0 (Math|Conversions|ToFloat(Integer) _returnvalue_2))
  (Rendering|Material|SetCustomPrimitiveDataFloat _sm_jamplayspace_border_square 1 (Math|Conversions|ToFloat(Integer) (stale EventOnTempoChanged NewTempo)))
  (Rendering|Material|SetCustomPrimitiveDataFloat _sm_jamplayspace_border_square 3 (Math|Conversions|ToFloat(Integer) (Mojo|JamSession|CountActiveSlotswithParams (Class|JamSessionInterface|GetMusicManager _asbp_jam_playspace) _jammusicslotsearchparams)))
  (Default|CallOnBeatStarted Beat%20in%20Bar%20%281-4%29 Measure%20Number%20%281-32%29 (Mojo|JamSession|CountActiveSlotswithParams (Class|JamSessionInterface|GetMusicManager _asbp_jam_playspace) _jammusicslotsearchparams)))

(event EventOnTempoChanged
  (params (NewTempo :type Integer))
  (Default|CallOnTempoChange NewTempo)
  (Variables|Default|SetTempo NewTempo))

(fn GetCenterLocation
  (return (Transformation|GetWorldLocation (Variables|Default|GetInner%20Collider))))

(fn GetInnerCollider
  (return (Variables|Default|GetInner%20Collider)))

(fn GetOuterCollider
  (return (Variables|Default|GetCollider)))

(fn IsPersistent
  (return true))

(fn ConstructionScript
  (|CallParentFunction))
